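# Physics asset for the female medium Holiday PJs face accessory hat, giving its dynamic hat bones simulated bodies.
# A larger capsule on the master hat bone anchors a chain of three smaller capsule bodies along the successive hat bones.
# Constraints link each bone to the previous one with limited swing and zero twist limit, so the hat can bend and flop without twisting about its length.
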PhysicsAsset /BRCosmetics/Accessories/Hats/F_MED_HolidayPJs_FaceAcc/Meshes/F_MED_Holiday_PJs_1_FaceAcc_Physics
bodies {
 dyn_hat_01 [{
   shapeType=Capsule
   length=3.1202
   radius=3.9032
   shapeName=dyn_hat_01_capsule
   center=<4.1182 0.4226 1.8695>
   rotation=<rot 80 0.1955 -0.0001>
  }]
 dyn_hat_01_MSTR [{
   shapeType=Capsule
   length=2.9511
   radius=9.7377
   shapeName=dyn_hat_01_MSTR_capsule
   center=<-0.2139 -4.51 -11.026>
   rotation=<rot 22.2461 -90 -100.9973>
  }]
 dyn_hat_02 [{shapeType=Capsule, length=3.1286, radius=1.5142, shapeName=dyn_hat_02_capsule, center=<3.2669 0 0>, rotation=<rot 90 0 0>}]
 dyn_hat_03 [{
   shapeType=Capsule
   length=0.0398
   radius=3.3513
   shapeName=dyn_hat_03_capsule
   center=<3.03 -0.814 -0.909>
   rotation=<rot 0 -89.8173 -173.3029>
  }]
}
constraints [
 {bone1Name=dyn_hat_02, bone2Name=dyn_hat_01, swing1LimitDegrees=25, swing2LimitDegrees=50, twistLimitDegrees=0}
 {bone1Name=dyn_hat_03, bone2Name=dyn_hat_02, swing1LimitDegrees=50, swing2LimitDegrees=50, twistLimitDegrees=0}
 {bone1Name=dyn_hat_01, bone2Name=dyn_hat_01_MSTR, swing1LimitDegrees=30, swing2LimitDegrees=50, twistLimitDegrees=0}
]
