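# Physics asset for the female medium Holiday PJs 3 face accessory, setting up simulated motion for its hair strands and bun.
# Box bodies on the hair bones and a capsule bun hang from the head capsule, either directly or in short multi-bone chains, with tight swing and twist limits.
# Two right-side chain roots lock one swing axis.
# An unconstrained spine_05 box supplies an upper-body shape alongside the head.
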PhysicsAsset /BRCosmetics/Accessories/Hats/F_MED_HolidayPJs_FaceAcc/Meshes/F_MED_Holiday_PJs_3_FaceAcc_Physics
bodies {
 dyn_bun [{shapeType=Capsule, length=3.25, radius=2.755, center=<-0.6349 0.0002 -0.0001>, rotation=<rot 0 -90 89.9994>}]
 dyn_hair_l_00 [{shapeType=Box, extentX=4.01, extentY=3.01, extentZ=2.01, center=<1.3618 0 0>}]
 dyn_hair_l_01 [{shapeType=Box, extentX=2.51, extentY=3.51, extentZ=3.01, center=<1.0956 -0.016 0.0792>, rotation=<rot 0 -89.9998 -4.1362>}]
 dyn_hair_l_02 [{shapeType=Box, extentX=2.5507, extentY=3.247, extentZ=4.6459, center=<2.1279 0.2734 0.8193>, rotation=<rot 2.0622 92.7932 -77.3>}]
 dyn_hair_l_04 [{shapeType=Box, extentX=1.9538, extentY=1.975, extentZ=6.028, center=<2.2208 0.2275 0.3678>, rotation=<rot 0 95.8503 -80.6435>}]
 dyn_hair_l_05 [{shapeType=Box, extentX=2.01, extentY=4.51, extentZ=2.01, center=<1.3716 0 0>, rotation=<rot 0 -90 0>}]
 dyn_hair_l_07 [{shapeType=Box, extentX=1.2407, extentY=1.5963, extentZ=5.0567, center=<1.7832 -0.2837 0.1749>, rotation=<rot -5.8359 -102.8667 -118.1797>}]
 dyn_hair_l_08 [{shapeType=Box, extentX=1.7407, extentY=1.5963, extentZ=3.0567, center=<0.6291 -0.0195 0.2842>, rotation=<rot -0.7434 -91.4422 -114.308>}]
 dyn_hair_l_09 [{shapeType=Box, extentX=1.01, extentY=2.51, extentZ=1.01, center=<0.6012 0 0.3604>, rotation=<rot 0 -89.9999 -30.9382>}]
 dyn_hair_l_10 [{shapeType=Box, extentX=1.4437, extentY=1.5525, extentZ=2.9839, center=<0.3921 0.0088 -0.248>, rotation=<rot -1.5726 88.7296 -113.3352>}]
 dyn_hair_r [{shapeType=Box, extentX=3.51, extentY=3.01, extentZ=2.01, center=<1.0591 0 0>}]
 dyn_hair_r_00 [{shapeType=Box, extentX=3.51, extentY=3.01, extentZ=2.01, center=<0.7576 0.4301 0>}]
 dyn_hair_r_01 [{shapeType=Box, extentX=2.9781, extentY=2.194, extentZ=2.5092, center=<1.1305 0 -0.0009>, rotation=<rot -0.0491 0 0>}]
 dyn_hair_r_02 [{shapeType=Box, extentX=2.9781, extentY=2.194, extentZ=2.5092, center=<1.2177 -0.2818 0.3895>, rotation=<rot -0.0491 0 0>}]
 dyn_hair_r_03 [{shapeType=Box, extentX=4.4781, extentY=2.194, extentZ=2.5092, center=<1.7508 0 -0.0015>, rotation=<rot -0.0491 0 0>}]
 dyn_hair_r_04 [{shapeType=Box, extentX=3.5782, extentY=1.5282, extentZ=1.3835, center=<1.5442 0 -0.417>, rotation=<rot -15.1091 0 0>}]
 dyn_hair_r_05 [{shapeType=Box, extentX=4.51, extentY=2, extentZ=1.51, center=<1.8383 0.0725 -0.0001>, rotation=<rot 0 -1.645 -0.0001>}]
 dyn_hair_r_06 [{shapeType=Box, extentX=2.01, extentY=3.51, extentZ=1.51, center=<0.9571 0 0>, rotation=<rot 0 -90 0>}]
 dyn_hair_r_07 [{shapeType=Box, extentX=4.5373, extentY=2.2006, extentZ=2.4022, center=<1.7016 0 0>}]
 dyn_hair_r_08 [{shapeType=Box, extentX=3.0373, extentY=2.7006, extentZ=4.9022, center=<1.5855 -0.7863 0.0297>, rotation=<rot 0.7056 63.6326 -89.0332>}]
 dyn_hair_r_09 [{shapeType=Box, extentX=3.51, extentY=3.01, extentZ=2.01, center=<1.0591 0 0>}]
 head [{shapeType=Capsule, length=9.5751, radius=4.6365, center=<1.5578 3.4018 -0.0929>, rotation=<rot 0.0664 -111.7455 -88.6158>}]
 spine_05 [{shapeType=Box, extentX=38.5, extentY=7.5, extentZ=37.5, center=<2.7944 1.9207 0>, rotation=<rot 0 -52.0774 0>}]
}
constraints [
 {bone1Name=dyn_bun, bone2Name=head, swing1LimitDegrees=10, swing2LimitDegrees=10, twistLimitDegrees=3}
 {bone1Name=dyn_hair_l_04, bone2Name=head, swing1LimitDegrees=30, swing2LimitDegrees=35, twistLimitDegrees=3}
 {bone1Name=dyn_hair_l_05, bone2Name=head, swing1LimitDegrees=30, swing2LimitDegrees=20, twistLimitDegrees=2}
 {bone1Name=dyn_hair_r_04, bone2Name=head, swing1LimitDegrees=25, swing2LimitDegrees=20, twistLimitDegrees=3}
 {bone1Name=dyn_hair_r_05, bone2Name=head, swing1LimitDegrees=30, swing2LimitDegrees=25, twistLimitDegrees=3}
 {bone1Name=dyn_hair_r_06, bone2Name=head, swing1LimitDegrees=25, swing2LimitDegrees=25, twistLimitDegrees=3}
 {bone1Name=dyn_hair_r_01, bone2Name=dyn_hair_r_00, swing1LimitDegrees=15, swing2LimitDegrees=25, twistLimitDegrees=3}
 {bone1Name=dyn_hair_r_02, bone2Name=dyn_hair_r_01, swing1LimitDegrees=20, swing2LimitDegrees=20, twistLimitDegrees=3}
 {bone1Name=dyn_hair_r_03, bone2Name=dyn_hair_r_02, swing1LimitDegrees=10, swing2LimitDegrees=10, twistLimitDegrees=3}
 {bone1Name=dyn_hair_l_01, bone2Name=dyn_hair_l_00, swing1LimitDegrees=30, swing2LimitDegrees=30, twistLimitDegrees=3}
 {bone1Name=dyn_hair_l_02, bone2Name=dyn_hair_l_01, swing1LimitDegrees=15, swing2LimitDegrees=15, twistLimitDegrees=3}
 {bone1Name=dyn_hair_r_07, bone2Name=dyn_hair_r_09, swing1LimitDegrees=25, swing2LimitDegrees=25, twistLimitDegrees=5}
 {bone1Name=dyn_hair_l_07, bone2Name=head, swing1LimitDegrees=20, swing2LimitDegrees=20, twistLimitDegrees=3}
 {bone1Name=dyn_hair_l_08, bone2Name=head, swing1LimitDegrees=30, swing2LimitDegrees=30, twistLimitDegrees=3}
 {bone1Name=dyn_hair_r_08, bone2Name=head, swing1LimitDegrees=25, swing2LimitDegrees=15, twistLimitDegrees=3}
 {bone1Name=dyn_hair_l_10, bone2Name=head, swing1LimitDegrees=5, swing2LimitDegrees=5, twistLimitDegrees=3}
 {bone1Name=dyn_hair_l_09, bone2Name=head, swing1LimitDegrees=25, swing2LimitDegrees=25, twistLimitDegrees=3}
 {bone1Name=dyn_hair_r_00, bone2Name=dyn_hair_r, swing1LimitDegrees=13, swing2LimitDegrees=13, twistLimitDegrees=1}
 {bone1Name=dyn_hair_r, bone2Name=head, swing1LimitDegrees=15, swing2LimitDegrees=15, swing2Motion=Locked, twistLimitDegrees=1}
 {bone1Name=dyn_hair_r_09, bone2Name=head, swing1LimitDegrees=15, swing2LimitDegrees=15, swing2Motion=Locked, twistLimitDegrees=1}
 {bone1Name=dyn_hair_l_00, bone2Name=head, swing1LimitDegrees=30, swing2LimitDegrees=10, twistLimitDegrees=3}
]
