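# Post-process style AnimBP for the Holiday PJs 3 female hat accessory, copying the attached parent's pose and adding secondary physics motion.
# Two RigidBody passes use different physics assets; the first lowers its alpha at high run speed, the second takes state-driven gravity.
# Additive offsets on the hat master bone blend in for skydiving between passes, and for nonzero vertical velocity afterward.
# The skydiving wind-vector logic is not wired to execution.
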AnimBlueprint /BRCosmetics/Accessories/Hats/F_MED_HolidayPJs_FaceAcc/Meshes/F_MED_Holiday_PJs_3_FaceAcc_AnimBP
target_skeleton=&/F_MED_Holiday_PJs_1_FaceAcc_Hat_LOD0_Skeleton
graphs [
 AnimGraph {nodes [
   n1 = RigidBody(ComponentPose=Local To Component(LocalPose=Copy Pose From Mesh() {bUseAttachedParent=true}), Alpha=Get RBDAlpha() -> (RBDAlpha:double)) {
    bClampLinearTranslationLimitToRefPose=true
    bOverrideWorldGravity=true
    overridePhysicsAsset=&/F_MED_Holiday_PJs_3_FaceAcc_Physics
    simulationSpace=BaseBoneSpace
    alphaBoolBlend {blendOption=Linear}
    baseBoneRef {boneName=neck_02}
    componentAppliedLinearAccClamp=<4000 4000 4000>
    componentLinearAccScale=<0.25 0.25 0.25>
    componentLinearVelScale=<0.25 0.25 0.25>
    overrideWorldGravity {z=-980}
   }
   n2 = "Transform (Modify) Bone - Bone: dyn_hat_01_MSTR"!(ComponentPose=@n1, bAlphaBoolEnabled:bool=Get bIsSkyDiving(), Translation="1.6,0,6.2", Rotation="-20,0,0", Scale="1,1,1") -> (Pose) {
    alphaInputType=Bool
    rotationMode=BMM_Additive
    rotationSpace=BCS_BoneSpace
    translationMode=BMM_Additive
    translationSpace=BCS_BoneSpace
    alphaBoolBlend {blendOption=Linear, blendOutTime=0.5}
    boneToModify {boneName=dyn_hat_01_MSTR}
    rotation {pitch=-20}
    translation {x=1.6, z=6.2}
   }
   n3 = RigidBody!(ComponentPose=@n2, Alpha=0.5, OverrideWorldGravity:Vector=Get GravityOverride(), ExternalForce) -> (Pose) {
    bOverrideWorldGravity=true
    overridePhysicsAsset=&/F_MED_Holiday_PJs_1_FaceAcc_Physics
    alphaBoolBlend {blendOption=Linear}
    componentLinearAccScale=<0.25 0.05 0.05>
    componentLinearVelScale=<0.25 0.05 0.05>
    overrideWorldGravity {z=-980}
   }
   n4 = "Transform (Modify) Bone - Bone: dyn_hat_01_MSTR"(ComponentPose=@n3, bAlphaBoolEnabled=Get JumpCheck() -> (JumpCheck:bool), Translation="1.6,0,6", Rotation="0,0,0", Scale="1,1,1") {
    alphaInputType=Bool
    rotationMode=BMM_Additive
    rotationSpace=BCS_BoneSpace
    scaleMode=BMM_Additive
    translationMode=BMM_Additive
    translationSpace=BCS_BoneSpace
    alphaBoolBlend {blendInTime=0.3, blendOutTime=0.3}
    boneToModify {boneName=dyn_hat_01_MSTR}
    translation {x=1.6, z=6}
   }
   Output Pose(Result=Component To Local(ComponentPose=@n4)) {layerGroup=None}
  ]}
 EventGraph {nodes [
   Event BlueprintUpdateAnimation()
   > n1 = Sequence!(execute) -> (then_0, then_1, then_2:Exec)
   > n2 = Gravity Override State Function DEPRECATED(CharacterPartAnimInstance=Self-Reference(), GameTimeInSeconds=GetGameTimeInSeconds(), GravityOverrideParams=Get Gravity Override Params())
   > Set GravityOverride(GravityOverride=@n2)
   n3 = Get PawnSpeed2D()
   Set RBDAlpha(execute=@n1.then_1, RBDAlpha:double=SelectFloat(A=0.5, B=1, bPickA="float >= float"(A=@n3, B=420) {is=title K2Node_CallFunction})) -> (Output_Get:double)
   n4 = TryGetPawnOwner()
   n5 = Branch(
    execute=@n1.then_2
    Condition="Not Equal (Float)"(A=GetVelocity!(self:Actor Object Reference=@n4) -> (ReturnValue_X:double, ReturnValue_Y:double, ReturnValue_Z:double).ReturnValue_Z)
   )
   > Set JumpCheck(JumpCheck:bool=true) -> (Output_Get:bool)
   Set JumpCheck(execute=@n5.else, JumpCheck=false)
   TryGetPawnOwner()
   n6 = Break Fort Anim Input Skydiving External Force(FortAnimInput_SkydivingExternalForce=Get SkydivingExternalForceInput()) -> (
    bUseSkydivingVectorForce=~
    SkydivingVectorForce=~
    HeadToPelvisDirection=~
    FloatingMultiplier=~
    FloatingAdditive=~
    DivingMultiplier=~
    DivingAdditive=~
    ParachutingMultiplier=~
    ParachutingAdditive=~
   )
   n7 = Get bIsOnGround() -> (bIsOnGround:bool)
   n8 = Get bIsSkyDiving()
   Set windVector(windVector:Vector=SelectVector(A=vector * float(A=@n6.SkydivingVectorForce, B=0.6), B="0, 0,-980", bPickA=OR Boolean(A=@n8, B=@n7))) -> (Output_Get:Vector)
  ]}
]
variables [
 Gravity Override Params {
  type=GravityOverrideParamsStruct
  default {onGround_Moving {gravityOverride=<0 0 -700>}, ziplining {gravityOverride=<0 0 -850>, windAmplitude=<200 50 0>, windFrequency=<5 5 0>}}
 }
 GravityOverride {type=Vector}
 JumpCheck {type=boolean}
 RBDAlpha {type=number}
 windVector {type=Vector}
]
